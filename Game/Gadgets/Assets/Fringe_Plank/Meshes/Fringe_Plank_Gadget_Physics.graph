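# Physics asset for the Fringe Plank gadget mesh, giving its DRing and Handle bones simple collision bodies.
# Each body is a single box: a thin, rotated, offset one for the D-ring and a small cube for the handle.
# One constraint joins the two bones with both swing axes locked, so the joint acts as a hinge about the twist axis.
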PhysicsAsset /Game/Gadgets/Assets/Fringe_Plank/Meshes/Fringe_Plank_Gadget_Physics
bodies {
 DRing [{shapeType=Box, extentX=1.01, extentY=5.4596, extentZ=6.4303, center=<-0.5292 -0.0008 2.7134>, rotation=<rot -0.0007 -90 -0.001>}]
 Handle [{shapeType=Box, extentX=1.1, extentY=1.1, extentZ=1.1}]
}
constraints [{
  bone1Name=DRing
  bone2Name=Handle
  swing1LimitDegrees=45
  swing1Motion=Locked
  swing2LimitDegrees=45
  swing2Motion=Locked
  twistLimitDegrees=60
 }]
